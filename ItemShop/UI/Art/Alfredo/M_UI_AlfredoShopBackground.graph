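# UI material for the Alfredo item shop background, with an animated "Headband" pattern, linear gradients and metal texture blends.
# The pattern multiplies two panning noise layers, adds a panning starburst scatter on rotated UVs, then darkens it by a gradient.
# A smoothstep of that result indexes a color curve atlas row, which blends with a mask texture.
# Named reroutes share the gradient and aspect-normalized UVs, with optional perspective distortion.
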Material /ItemShop/UI/Art/Alfredo/M_UI_AlfredoShopBackground
graph {
 properties {bUsedWithStaticMesh=false, materialDomain=MD_UI, pixelDepthOffsetMode=PDOM_Legacy, usedShadingModels=Default Lit}
 nodes [
  n1 = Constant()
  n2 = Constant() {r=0.5}
  n3 = ConstantBiasScale(None=TextureCoordinate()) {bias=-0.5, scale=2}
  n4 = Subtract(A=ComponentMask(None=Abs(None=@n3)) {g=true}, B=ScalarParameter() {defaultValue=0.2869, parameterName=Subtract})
  n5 = ScalarParameter() {defaultValue=0.0659, parameterName=Falloff}
  n6 = Multiply(A=Saturate(None=Multiply(A=@n4, B=@n5)), B=ScalarParameter() {defaultValue=8.2497, parameterName=Intensity})
  n7 = VectorParameter() {parameterName=RGB2, defaultValue=<rgba 0 0.2825 1 1>}
  n8 = Constant() {r=-0.4}
  n9 = NamedRerouteUsage()
  n10 = Constant3Vector() {constant=<rgba 0 0.8388 1 0>}
  Add(
   A=LinearInterpolate(
    A=Constant3Vector() {constant=<rgba 0 0.0259 0.3802 0>}
    B=Constant3Vector() {constant=<rgba 0 0.3649 0.5625 0>}
    Alpha=MaterialFunctionCall(GradientStart=@n1, GradientEnd=@n2) {materialFunction=&MF_UI_LinearGradient}.Result
   )
   B=LinearInterpolate(A=Multiply(A=@n6, B=@n7.RGB), B=Constant() {r=0.928}, Alpha=LinearInterpolate(A=@n8, B=@n9, Alpha=@n10))
  )
  Constant() {r=1.0542}
  n11 = TextureSample() {texture=&/T_UI_Alfredo_Metal}
  n12 = VectorParameter() {parameterName=RGB3, defaultValue=<rgba 0.1977 0.2199 0.3333 1>}
  n13 = Constant()
  n14 = Constant() {r=0.55}
  n15 = Constant() {r=0.02}
  n16 = Saturate(None=OneMinus(None=MaterialFunctionCall(GradientStart=@n13, GradientEnd=@n14, GradientOffset=@n15) {materialFunction=&MF_UI_LinearGradient}.Result))
  n17 = NamedRerouteUsage()
  n18 = Constant() {r=0.163}
  n19 = TextureSample() {texture=&T_UI_Alfredo_Metal}
  LinearInterpolate(
   A=LinearInterpolate(A=@n11.RGB, B=LinearInterpolate(B=@n12.RGB, Alpha=@n16), Alpha=Multiply(A=@n11.B, B=@n16))
   Alpha=LinearInterpolate(A=Constant() {r=-0.4}, B=LinearInterpolate(A=@n17, B=@n18, Alpha=@n19.RGB))
  ) {constB=0.632}
  n20 = MaterialFunctionCall(Float2=MaterialFunctionCall() {materialFunction=&ScreenResolution}.Visible Resolution) {materialFunction=&BreakOutFloat2Components}
  n21 = Max(A=@n20.G)
  Multiply(
   A=AppendVector(A=Divide(A=@n20.R, B=@n21), B=Constant() {r=1})
   B=MaterialFunctionCall() {materialFunction=&ScreenAlignedUVs}."X 100%, Y 100%"
  )
  n22 = ScalarParameter() {defaultValue=0.0857, group=Headband, parameterName=Min, sliderMax=1}
  n23 = ScalarParameter() {defaultValue=0.9524, group=Headband, parameterName=Max, sliderMax=1}
  n24 = VectorParameter() {group=Headband, parameterName=Streaks, channelNames {a="", b=Speed, g=StretchY, r=StretchX}, defaultValue=<rgba 0.7329 0.0052 0.005 1>}
  n25 = MaterialFunctionCall(UVs=NamedRerouteUsage(), "Rotation Angle (0-1)"=ScalarParameter() {defaultValue=0.7173, group=Headband, parameterName=Rotation Angle}) {materialFunction=&CustomRotator}
  n26 = Panner(
   Coordinate=Multiply(A=@n25.Rotated Values, B=AppendVector(A=@n24.StretchX, B=@n24.StretchX))
   Speed=AppendVector(A=Constant() {r=0.003}, B=@n24.Speed)
  )
  n27 = VectorParameter() {group=Headband, parameterName=Noise, channelNames {a=Power, b=Speed, g=StretchY, r=StretchX}, defaultValue=<rgba 3.6698 0.0826 0.01 -0.3543>}
  n28 = AppendVector(A=@n27.StretchX, B=@n27.StretchX)
  n29 = Constant() {r=0.003}
  n30 = TextureSampleParameter2D(UVs=Panner(Coordinate=Multiply(A=@n25.Rotated Values, B=@n28), Speed=AppendVector(A=@n29, B=@n27.Speed))) {
   group=Headband
   parameterName=NoisePattern
   samplerType=SAMPLERTYPE_LinearColor
   texture=&HighContrastNoise
  }
  n31 = Power(Base=@n30.R, Exp=@n27.Power)
  n32 = VectorParameter() {group=Headband, parameterName=Noise2, channelNames {a=Power, b=Speed, g=StretchY, r=StretchX}, defaultValue=<rgba 0.6505 0.0398 0.1 0.3605>}
  n33 = AppendVector(A=@n32.StretchX, B=@n32.StretchX)
  n34 = Constant()
  n35 = TextureSampleParameter2D(UVs=Panner(Coordinate=Multiply(A=@n25.Rotated Values, B=@n33), Speed=AppendVector(A=@n34, B=@n32.Speed))) {
   group=Headband
   parameterName=NoisePattern
   samplerType=SAMPLERTYPE_LinearColor
   texture=&HighContrastNoise
  }
  n36 = Power(Base=@n35.R, Exp=@n32.Power)
  n37 = Constant()
  n38 = ScalarParameter() {defaultValue=0.1714, group=Headband, parameterName=GradientOffset}
  n39 = ComponentMask(None=Add(A=TextureCoordinate(), B=AppendVector(A=@n37, B=@n38))) {g=true}
  n40 = ScalarParameter() {defaultValue=0.8777, group=Headband, parameterName=DarkenGradient}
  n41 = Multiply(
   A=Add(
    A=TextureSampleParameter2D(UVs=@n26) {group=Headband, parameterName=Pattern, texture=/Game/Global/Textures/Default/PointSplatter/T_UI_StarBurst_Scatter}.R
    B=Multiply(A=@n31, B=@n36)
   )
   B=OneMinus(None=Multiply(A=@n39, B=@n40))
  )
  n42 = CurveAtlasRowParameter(CurveTime=ComponentMask(None=SmoothStep(Min=@n22, Max=@n23, Value=@n41)) {r=true}) {
   atlas=/Game/UI/Frontend/ItemShop/Art/ItemShopColorAtlas
   curve=/Game/UI/Frontend/ItemShop/Art/Alfredo/ItemShop_AlfredoBG_Curve
   defaultValue=8
   parameterName=CurveParamsBG
   sortPriority=0
  }
  NamedRerouteDeclaration(None=LinearInterpolate(A=@n42, B=TextureSample() {texture=&/T_UI_Alfredo_Mask}.RGB) {constAlpha=-0.316}) {name=Gradient_PY, nodeColor=<rgba 0 0 0 1>}
  n43 = VectorParameter() {group=Headband, parameterName=HeadbandScale, channelNames {a=OffsetY, b=OffsetX, g=Y, r=X}, defaultValue=<rgba 1 1 0 0>}
  n44 = MaterialFunctionCall!(
   Texture Scale=AppendVector(A=@n43.X, B=@n43.X)
   UVs=NamedRerouteUsage()
   Offset XY Texturesize Z=MaterialFunctionCall(R=@n43.OffsetX, G=@n43.OffsetX, B=Constant() {r=1}) {materialFunction=&AppendMany}.RGB
  ) {materialFunction=/Engine/Functions/Engine_MaterialFunctions01/Texturing/ScaleUVsAroundPoint}
  n45 = VectorParameter() {group=Headband, parameterName=HeadbandPerspective, channelNames {a=YSlide, b=XSlide, g=Y, r=X}}
  n46 = MaterialFunctionCall!(UV=@n44.UVs, Texture, FixedScaleFactor_X, FixedScaleFactor_Y, Persp_Y=@n45.Y, Persp_X=@n45.Y, Slide_X=@n45.Y, Slide_Y=@n45.Y, Flip_Y, Flip_X) {materialFunction=/Game/UI/Foundation/Materials/MF_UI_BasicPerspective}
  NamedRerouteDeclaration(None=Add(A=StaticSwitchParameter(True=@n46.FinalUvs, False=@n44.UVs) {group=Headband, parameterName=HeadbandUsePerpsective}) {constB=0}) {name=NormalizedSSUVs, nodeColor=<rgba 1 0 0.7969 1>}
  n47 = MaterialFunctionCall(Float2=MaterialFunctionCall() {materialFunction=&GetUserInterfaceUV}.Pixel Size) {materialFunction=&BreakOutFloat2Components}
  n48 = Divide(A=@n47.G, B=@n47.G)
  n49 = Constant() {r=1}
  NamedRerouteDeclaration(None=MaterialFunctionCall(UVs=TextureCoordinate(), Texture Scale=AppendVector(A=@n48, B=@n49)) {materialFunction=&ScaleUVsByCenter}.UVs) {name=NormalizedUVs, nodeColor=<rgba 0.8047 1 0 1>}
  Reroute()
  TextureCoordinate()
 ]
}
