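# UI material for an Alfredo item shop tile, built on a blue emissive palette with feathered edge-frame glows.
# One branch layers panning squares and cell noise over a radial gradient; another adds scrolling squares, a sweep line and a frame texture distorted by glitching static.
# A third blends square-masked MF_UI_Confetti with a radial color gradient.
# These branches end separately with no material output link shown, alongside unused pixel-inset math.
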Material /ItemShop/UI/Art/Alfredo/M_UI_AlfredoShopTIle_02
graph {
 properties {bUsedWithStaticMesh=false, materialDomain=MD_UI, pixelDepthOffsetMode=PDOM_Legacy, usedShadingModels=Default Lit}
 nodes [
  n1 = MaterialFunctionCall() {materialFunction=&GetUserInterfaceUV}
  n2 = Subtract(A=@n1.Pixel Size, B=Multiply(A=ScalarParameter() {defaultValue=6, parameterName=PixelSize}) {constB=2})
  Divide(A=ComponentMask(None=@n2) {r=true}, B=ComponentMask(None=@n1.Pixel Size) {r=true})
  Divide(A=ComponentMask(None=@n2) {g=true}, B=ComponentMask(None=@n1.Pixel Size) {g=true})
  n3 = MaterialFunctionCall() {materialFunction=&GetUserInterfaceUV}
  n4 = Subtract(A=@n3.Pixel Size, B=Multiply(A=ScalarParameter() {defaultValue=6, parameterName=PixelSize}) {constB=2})
  Divide(A=ComponentMask(None=@n4) {r=true}, B=ComponentMask(None=@n3.Pixel Size) {r=true})
  Divide(A=ComponentMask(None=@n4) {g=true}, B=ComponentMask(None=@n3.Pixel Size) {g=true})
  n5 = MaterialFunctionCall() {materialFunction=&GetUserInterfaceUV}
  n6 = Subtract(A=@n5.Pixel Size, B=Multiply(A=ScalarParameter() {defaultValue=6, parameterName=PixelSize}) {constB=2})
  Divide(A=ComponentMask(None=@n6) {r=true}, B=ComponentMask(None=@n5.Pixel Size) {r=true})
  Divide(A=ComponentMask(None=@n6) {g=true}, B=ComponentMask(None=@n5.Pixel Size) {g=true})
  n7 = MaterialFunctionCall() {materialFunction=&GetUserInterfaceUV}
  n8 = Subtract(A=@n7.Pixel Size, B=Multiply(A=ScalarParameter() {defaultValue=6, parameterName=PixelSize}) {constB=2})
  Divide(A=ComponentMask(None=@n8) {r=true}, B=ComponentMask(None=@n7.Pixel Size) {r=true})
  Divide(A=ComponentMask(None=@n8) {g=true}, B=ComponentMask(None=@n7.Pixel Size) {g=true})
  n9 = Multiply(
   A=AppendVector(A=ScalarParameter() {defaultValue=8, parameterName="U_Tiling(C)"}, B=ScalarParameter() {defaultValue=4, parameterName="V_Tiling(C)"})
   B=MaterialFunctionCall() {materialFunction=&MF_UI_UVScaling}.Scaled to Fit
  )
  n10 = TextureSampleParameter2D!(
   UVs=MaterialFunctionCall(UVs=Panner(Coordinate=@n9) {speedX=0.5}, "Rotation Angle (0-1)"=Constant()) {materialFunction=&CustomRotator}.Rotated Values
   Level
   Apply View MipBias
  ) {mipValueMode=TMVM_MipLevel, parameterName=Texture_Input_C, texture=/Game/Effects/Fort_Effects/Textures/Tile/Tests/T_Tile_Panner_Mask}
  n11 = Add(A=@n10.G, B=ScalarParameter() {defaultValue=0.1, group=Tech, parameterName=Tech Noise Add})
  n12 = AppendVector(A=ScalarParameter() {defaultValue=7, parameterName="U_Tiling(B)"}, B=ScalarParameter() {defaultValue=7, parameterName="V_Tiling(B)"})
  n13 = MaterialFunctionCall() {materialFunction=&MF_UI_UVScaling}
  n14 = Multiply(
   A=Multiply(A=Power(Base=@n11), B=ScalarParameter() {defaultValue=2, parameterName=Texture_Multiply_C}) {constB=2}
   B=TextureSampleParameter2D!(UVs=Multiply(A=@n12, B=@n13.ScaleY), Apply View MipBias) {parameterName=Texture_Input_B, texture=/OfferCatalog/Art/Alfredo/T_UI_Squares}.R
  )
  n15 = MaterialFunctionCall() {materialFunction=&MF_UI_UVScaling}
  n16 = Constant2Vector() {g=-0.025, r=-0.025}
  n17 = Constant2Vector() {g=0.025}
  n18 = Multiply(
   A=TextureSample(UVs=Panner(Coordinate=@n15.Scaled to Fit, Speed=@n16)) {samplerType=SAMPLERTYPE_Grayscale, texture=/Game/Weapons/FORT_Melee/Pickaxe_Cube_Ninja_Male/FX/T_CellNoise}.RGB
   B=TextureSample(UVs=Panner(Coordinate=@n15.Scaled to Fit, Speed=@n17) {constCoordinate=1}) {samplerType=SAMPLERTYPE_Grayscale, texture=&T_CellNoise}.RGB
  )
  n19 = ScalarParameter() {defaultValue=0.9017, parameterName=FrameFeather3}
  n20 = ScalarParameter() {defaultValue=1.3707, parameterName=FrameFeather4}
  n21 = Abs(None=ConstantBiasScale(None=TextureCoordinate()) {bias=-0.5, scale=2})
  n22 = SmoothStep(Min=@n19, Max=@n20, Value=ComponentMask(None=@n21) {r=true})
  n23 = SmoothStep(Min=@n19, Max=@n20, Value=ComponentMask(None=@n21) {g=true})
  n24 = Multiply(A=Saturate(None=Add(A=@n22, B=@n23)), B=ScalarParameter() {defaultValue=2.6917, parameterName=FrameOpacity})
  n25 = VectorParameter() {parameterName=Emissive Color, defaultValue=<rgba 0 0.6084 1 0>}
  n26 = MaterialFunctionCall(
   UVs=MaterialFunctionCall() {materialFunction=&MF_UI_UVScaling}.ScaleY
   CenterPosition=Constant2Vector() {g=0.4, r=0.5}
   Radius=ScalarParameter() {defaultValue=0.536, parameterName=Radius}
   Density=ScalarParameter() {defaultValue=1.6537, parameterName=Density}
  ) {materialFunction=&RadialGradientExponential}
  n27 = VectorParameter() {parameterName=Emissive Colorradial, defaultValue=<rgba 0 0.1719 0.7682 1>}
  LinearInterpolate(
   A=LinearInterpolate(A=Multiply(A=@n24, B=@n25.RGB), B=Multiply(A=@n26.RadialGradientExponential, B=@n27.RGB))
   B=LinearInterpolate(B=VectorParameter() {parameterName=EmissiveTint, defaultValue=<rgba 0.2718 4.9041 15 1>}.RGB, Alpha=Multiply(A=@n14, B=@n18))
  )
  n28 = ScalarParameter() {defaultValue=0.9017, parameterName=FrameFeather1}
  n29 = ScalarParameter() {defaultValue=1.8432, parameterName=FrameFeather2}
  n30 = Abs(None=ConstantBiasScale(None=TextureCoordinate()) {bias=-0.5, scale=2})
  n31 = ComponentMask(None=@n30) {r=true}
  n32 = ComponentMask(None=@n30) {g=true}
  n33 = Saturate(None=Add(A=SmoothStep(Min=@n28, Max=@n29, Value=@n31), B=SmoothStep(Min=@n28, Max=@n29, Value=@n32)))
  n34 = ScalarParameter() {defaultValue=2.6917, parameterName=FrameOpacity}
  n35 = VectorParameter() {parameterName=Emissive Color, defaultValue=<rgba 0 0.6084 1 0>}
  n36 = VectorParameter() {parameterName=BG Color, defaultValue=<rgba 0 0 0 1>}
  n37 = Panner(Coordinate=TextureCoordinate()) {speedY=-0.015}
  n38 = ScalarParameter() {defaultValue=1.5}
  n39 = Multiply(
   A=VectorParameter() {parameterName=EmissiveTint, defaultValue=<rgba 0.2718 4.9041 15 1>}.RGB
   B=TextureSampleParameter2D(UVs=Multiply(A=@n37, B=@n38)) {parameterName=Param, texture=&T_UI_Squares}.R
  )
  n40 = ComponentMask(None=TextureCoordinate()) {g=true}
  n41 = Time()
  n42 = ScalarParameter() {defaultValue=-0.25, parameterName=SwipeTimeSkew}
  n43 = Frac(None=Add(A=Multiply(A=@n40) {constB=0.75}, B=Multiply(A=@n41, B=@n42) {constB=-0.2}))
  n44 = Multiply(A=Multiply(A=Power(Base=@n43) {constExponent=20}, B=ScalarParameter() {defaultValue=1, group=Sweep Line, parameterName=Sweep Add}) {constB=0.2})
  n45 = Multiply(A=@n35.RGB, B=@n44)
  n46 = TextureCoordinate()
  n47 = ScalarParameter() {defaultValue=0.344, group=Static, parameterName=StaticLowValue}
  n48 = ScalarParameter() {defaultValue=1, group=Static, parameterName=StaticHighValue}
  n49 = TextureCoordinate()
  n50 = Multiply(A=Time(), B=ScalarParameter() {defaultValue=0.08, parameterName=BigGlitchSpeed})
  n51 = Round(None=Power(Base=Frac(None=@n50)))
  n52 = LinearInterpolate(A=@n49, B=Multiply(A=@n49) {constB=0.75}, Alpha=@n51)
  n53 = TextureSample(UVs=Panner(Coordinate=@n52) {speedX=5, speedY=0.1}) {samplerType=SAMPLERTYPE_Masks, texture=/Game/Missions/Primary/RepairTheShelter/Textures/Cinematic_Screen_Static}
  n54 = Multiply(
   A=TextureSample(UVs=Panner(Coordinate=@n52) {speedX=6, speedY=-0.05}) {samplerType=SAMPLERTYPE_Masks, texture=&Cinematic_Screen_Static}.G
   B=TextureSample(UVs=Panner(Coordinate=@n52) {speedX=-3, speedY=-0.03}) {samplerType=SAMPLERTYPE_Masks, texture=&Cinematic_Screen_Static}.R
  )
  n55 = Multiply(
   A=Add(A=Multiply(A=@n53.B, B=@n54) {constB=0.1}, B=@n54)
   B=LinearInterpolate(
    A=ScalarParameter() {defaultValue=6.2829, group=Static, parameterName=Static Scale}
    B=ScalarParameter() {defaultValue=10, group=Static, parameterName=Static Big Scale}
   )
  )
  n56 = LinearInterpolate(
   A=ScalarParameter() {defaultValue=0.7, group=Static, parameterName=Static UV Distort}
   B=ScalarParameter() {defaultValue=9, group=Static, parameterName=Static UV Big Distort}
   Alpha=@n51
  )
  n57 = Multiply(A=Add(A=LinearInterpolate(A=@n47, B=@n48, Alpha=@n55), B=@n44), B=AppendVector(A=Multiply(A=@n54, B=@n56), B=Constant()))
  n58 = Multiply(A=TextureSample(UVs=Add(A=@n46, B=@n57)) {texture=/OfferCatalog/Art/Alfredo/T_UI_Frame}.R, B=@n35.RGB)
  LinearInterpolate(
   A=Multiply(A=Multiply(A=@n33, B=@n34), B=@n35.RGB)
   B=Add(A=LinearInterpolate(A=@n36.RGB, B=@n39, Alpha=@n39), B=LinearInterpolate(A=@n45, B=@n58) {constAlpha=0.8})
  )
  n59 = MaterialFunctionCall() {materialFunction=&MF_UI_UVScaling}
  n60 = AppendVector(A=ScalarParameter() {parameterName=L1-Patter-X-Offset}, B=ScalarParameter() {parameterName=L1-Patter-Y-Offset})
  n61 = ScalarParameter() {defaultValue=6.6886, parameterName=L1-Pattern-X}
  n62 = ScalarParameter() {defaultValue=2.904, parameterName=L1-Pattern-Y}
  n63 = ScalarParameter() {defaultValue=3.6311, parameterName=L1-Confetti-Local-Tile-X}
  n64 = ScalarParameter() {defaultValue=243.3965, parameterName=L1-Confetti-Local-Tile-Y}
  n65 = ScalarParameter() {defaultValue=37.6207, parameterName=L1-SmoothStep-Max}
  n66 = ScalarParameter() {defaultValue=180.8739, parameterName=L1-SmoothStep-Min}
  n67 = Multiply(
   A=MaterialFunctionCall!(
    "UV(V2)"=Add(A=@n59.Scaled to Fit, B=@n60)
    Global-Angle=ScalarParameter() {defaultValue=0.003, parameterName=L1-Global Angle}
    "Pattern-Tiles(V2)"=AppendVector(A=@n61, B=@n62)
    RandomSeed=ScalarParameter() {defaultValue=12345, parameterName=L1-RandomSeed}
    Confetti-Speed-TimeFactor=ScalarParameter() {defaultValue=0.1485, parameterName=L1-Confetti-Speed-TimeFactor}
    Confetti-Rotation-TimeFactor=ScalarParameter() {parameterName=L1-Confetti-Rotation-TimeFactor}
    Confetti-Local-Rotation-Range=ScalarParameter() {parameterName=L1-Confetti-Local-Rotation-Range}
    "Confetti-Local-Tiles(V2)"=AppendVector(A=@n63, B=@n64)
    "SmoothStepRanges(V2)"=AppendVector(A=@n65, B=@n66)
   ) {materialFunction=/Game/UI/Frontend/ItemShop/Art/Cactus/MF_UI_Confetti}.Confetti-Mask
   B=VectorParameter() {parameterName=EmissiveTint, defaultValue=<rgba 0.2718 4.9041 15 1>}.RGB
  )
  n68 = MaterialFunctionCall() {materialFunction=&MF_UI_UVScaling}
  n69 = TextureSampleParameter2D(UVs=Multiply(A=@n68.Scaled to Fit, B=ScalarParameter() {defaultValue=1.5})) {parameterName=Param, texture=&T_UI_Squares}
  n70 = Constant3Vector() {constant=<rgba 0.0284 0.1144 0.2502 0>}
  n71 = Constant3Vector() {constant=<rgba 0 0.5906 0.6514 0>}
  n72 = AppendVector(
   A=ScalarParameter() {defaultValue=-0.1939, parameterName=Radial-X-Tile}
   B=ScalarParameter() {defaultValue=0.5225, parameterName=Radial-Y-Tile}
  )
  n73 = Multiply(A=MaterialFunctionCall() {materialFunction=&MF_UI_UVScaling}.Scaled to Fit, B=@n72)
  n74 = Multiply(A=@n72, B=Constant2Vector() {g=1, r=0.5})
  n75 = Saturate(None=Power(Base=Distance(A=@n73, B=@n74), Exp=ScalarParameter() {defaultValue=1.8817, parameterName=Distance Power}))
  n76 = ScalarParameter() {defaultValue=0.9017, parameterName=FrameFeather1}
  n77 = ScalarParameter() {defaultValue=1.8432, parameterName=FrameFeather2}
  n78 = Abs(None=ConstantBiasScale(None=TextureCoordinate()) {bias=-0.5, scale=2})
  n79 = ComponentMask(None=@n78) {r=true}
  n80 = ComponentMask(None=@n78) {g=true}
  n81 = Saturate(None=Add(A=SmoothStep(Min=@n76, Max=@n77, Value=@n79), B=SmoothStep(Min=@n76, Max=@n77, Value=@n80)))
  n82 = ScalarParameter() {defaultValue=2.6917, parameterName=FrameOpacity}
  LinearInterpolate(
   A=Multiply(A=Multiply(A=@n81, B=@n82), B=VectorParameter() {parameterName=Emissive Color, defaultValue=<rgba 0 0.6084 1 0>}.RGB)
   B=LinearInterpolate(A=LinearInterpolate(B=@n67, Alpha=@n69.R), B=LinearInterpolate(A=@n70, B=@n71, Alpha=@n75))
  )
  n83 = ScalarParameter() {defaultValue=0.9017, parameterName=FrameFeather1}
  n84 = ScalarParameter() {defaultValue=1.8432, parameterName=FrameFeather2}
  n85 = Abs(None=ConstantBiasScale(None=TextureCoordinate()) {bias=-0.5, scale=2})
  n86 = SmoothStep(Min=@n83, Max=@n84, Value=ComponentMask(None=@n85) {r=true})
  n87 = SmoothStep(Min=@n83, Max=@n84, Value=ComponentMask(None=@n85) {g=true})
  Multiply(A=Saturate(None=Add(A=@n86, B=@n87)), B=ScalarParameter() {defaultValue=2.6917, parameterName=FrameOpacity})
  Panner(Coordinate=@n68.Scaled to Fit) {speedY=0.1}
 ]
}
